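# Material function that keeps a quad at constant on-screen size through world position offset, though it breaks down at close range.
# It remaps UVs to -1..1 and builds an XY corner direction, transformed with the default tangent-to-world transform.
# That vector is multiplied by the camera-to-world-position distance and the SizeScale input, so the offset grows linearly with distance.
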MaterialFunction /Engine/Functions/Engine_MaterialFunctions01/WorldPositionOffset/ConstantScalebyDistance
graph {
 properties {
  bExposeToLibrary=true
  description="Keeps a quad a constant scale on the screen, falls apart at close distances."
  libraryCategoriesText [Misc, World Position Offset]
 }
 nodes [
  n1 = Multiply(A=Add(A=TextureCoordinate()) {constB=-0.5}) {constB=2}
  n2 = Add(
   A=Multiply(A=Constant3Vector() {constant=<rgba 1 0 0 0>}, B=ComponentMask(None=@n1) {r=true})
   B=Multiply(A=ComponentMask(None=@n1) {g=true}, B=Constant3Vector() {constant=<rgba 0 1 0 0>})
  )
  n3 = Distance(A=WorldPosition().XYZ, B=CameraPositionWS())
  n4 = FunctionInput() {bUsePreviewValueAsDefault=true, inputName=SizeScale, inputType=FunctionInput_Scalar}
  FunctionOutput(None=Multiply(A=Transform(None=@n2), B=Multiply(A=@n3, B=@n4)) {constB=50})
 ]
}
