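# Translucent unlit storm-visual material that fakes decals on the scene behind a volume, recoloring it with animated two-color tinting.
# A pivot-distance falloff from world position behind translucency, broken up by scrolling world-aligned noise through HeightLerp and stencil-masked, feeds desaturated scene-color math.
# Offset math wobbles the volume per actor and collapses it toward its pivot as the camera nears.
# Material output connections are not shown, and several branches end unconsumed.
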Material /Game/VisualThreatManager/StormVisuals/Particles/ParticleBasedDecals_trimmed
graph {
 properties {
  bTangentSpaceNormal=false
  blendMode=BLEND_Translucent
  blendableLocation=BL_SceneColorAfterDOF
  pixelDepthOffsetMode=PDOM_Legacy
  shadingModel=MSM_Unlit
  usedShadingModels=Unlit
 }
 nodes [
  n1 = MaterialFunctionCall() {materialFunction=&ComponentPivotLocation}
  n2 = ActorPositionWS()
  n3 = Add(
   A=Multiply(A=Time()) {constB=0.1}
   B=Multiply(A=MaterialFunctionCall!(f2, f3=@n2, f4) {materialFunction=/Game/Packages/Fortnite_UniqueMaterials_1/VectorComponentOps}.f3 Add) {constB=0.005}
  )
  n4 = Add(A=@n3) {constB=0.15}
  n5 = Multiply(A=MaterialFunctionCall(X=Cosine(None=@n3), Y=Sine(None=@n3), Z=Sine(None=@n4) {period=3}) {materialFunction=&MakeFloat3}.Result) {constB=200}
  n6 = Subtract(
   A=Distance(A=CameraPositionWS(), B=Add(A=@n1.Result, B=@n5))
   B=Add(A=MaterialFunctionCall() {materialFunction=/Engine/Functions/Engine_MaterialFunctions02/WorldPositionOffset/ObjectScale}.Scale X) {constB=32, desc=extra buffer}
  )
  n7 = OneMinus(None=Clamp(None=Divide(A=@n6) {constB=512}))
  n8 = MaterialFunctionCall() {materialFunction=&ComponentPivotLocation}
  n9 = WorldPosition()
  Add(A=Multiply(A=@n7, B=Subtract(A=@n8.Result, B=@n9.XYZ)), B=@n5)
  Constant3Vector() {constant=<rgba 1 0 0 0>}
  Constant() {r=0.2}
  Constant()
  Frac(None=Time())
  n10 = MaterialFunctionCall(Exclude WPO Offsets=StaticBool() {value=true}) {materialFunction=&WorldPositionBehindTranslucency}
  n11 = MaterialFunctionCall() {materialFunction=&ObjectScale}
  n12 = Clamp(None=Divide(A=Distance(A=@n10.Result, B=@n1.Result), B=@n11.Scale X))
  n13 = Constant() {r=0.8}
  n14 = MaterialFunctionCall() {materialFunction=&WorldPositionBehindTranslucency}
  n15 = Constant() {r=0.01}
  n16 = SceneTexture() {sceneTextureId=PPI_WorldNormal}
  n17 = Constant3Vector() {constant=<rgba 0 0 1 0>}
  n18 = MaterialFunctionCall(
   TextureObject=TextureObject() {samplerType=SAMPLERTYPE_LinearColor, texture=/Game/VisualThreatManager/StormVisuals/Textures/testGrayPack}
   TextureSize=Constant() {r=1}
   WorldPosition=Add(
    A=Divide(A=@n14.Result) {constB=512}
    B=MaterialFunctionCall(Speed=@n15) {materialFunction=&TimeWithSpeedVariable}.Frac Time
   )
   World Space Normal=ShadingPathSwitch(Default=@n16.Color, Deferred=@n16.Color, Forward=@n17, Mobile=@n17)
  ) {materialFunction=&WorldAlignedTexture}
  n19 = ComponentMask(None=@n18.XYZ Texture) {r=true}
  n20 = MaterialFunctionCall(
   Transition Phase=Add(
    A=MaterialFunctionCall!(
     Texture=@n12
     ---------------
     New Black Value
     New Middle Value
     New White Value
     Middle Point=@n13
     "--------------- "
     Define Interpolation Curve
     Interpolation Power
     Invert Interpolation Power
    ) {materialFunction=/Engine/Functions/Engine_MaterialFunctions02/3PointLevels}.Result
    B=@n7
   )
   Height Texture=OneMinus(None=@n19)
   Contrast=Constant() {r=0.2}
  ) {materialFunction=&HeightLerp}
  n21 = MaterialFunctionCall(In=@n20.Alpha, Contrast=ScalarParameter() {defaultValue=0.5, parameterName=falloffGradientPower}) {materialFunction=&CheapContrast}
  n22 = CollectionParameter() {collection=/Game/Packages/Fortress_SharedMaterials/Stencil_ID_Definitions/Stencil_ID_Definitions, parameterName=Decal Mask}
  n23 = Multiply(
   A=Add(A=OneMinus(None=@n21.Result), B=OneMinus(None=@n20.Alpha))
   B=OneMinus(None=MaterialFunctionCall!(ID=@n22) {materialFunction=/Game/Packages/Fortress_SharedMaterials/Stencil_ID_Definitions/StencilMaskCompare}.Result)
  ) {desc=used to use the decal bit mask}
  LinearInterpolate(Alpha=Clamp(None=@n23)) {constA=1, constB=0.2}
  n24 = SceneColor!(UVs)
  n25 = Desaturation(None=@n24.RGB)
  MaterialFunctionCall(Emissive=ShadingPathSwitch(Default=Multiply(A=@n25) {constB=0.2}, Mobile=Constant() {r=0.1}) {desc=Migrated from FeatureLevelSwitch}) {materialFunction=&MF_FX_EmissiveAdjustments}
  MaterialFunctionCall() {materialFunction=&WorldPositionBehindTranslucency}
  MaterialFunctionCall!(SizeScale) {materialFunction=/Engine/Functions/Engine_MaterialFunctions01/WorldPositionOffset/ConstantScalebyDistance}
  MaterialFunctionCall() {materialFunction=&ComponentPivotLocation}
  n26 = OneMinus(None=Desaturation(None=SceneTexture() {sceneTextureId=PPI_DiffuseColor}.Color) {luminanceFactors=<rgba 1 1 1 0>})
  Multiply(A=Multiply(A=@n26, B=@n26), B=@n23)
  n27 = MaterialFunctionCall() {materialFunction=&WorldPositionBehindTranslucency}
  n28 = MaterialFunctionCall(Float3=Add(A=@n18.XYZ Texture, B=Divide(A=@n27.Result) {constB=1024}) {constB=0}) {materialFunction=&BreakOutFloat3Components}
  n29 = MaterialFunctionCall(Speed=Constant() {r=0.125}) {materialFunction=&TimeWithSpeedVariable}
  n30 = Add(A=Add(A=@n28.B, B=@n29.Time)) {constB=0.25}
  n31 = Add(A=@n28.G, B=@n29.Time)
  n32 = Multiply(A=ConstantBiasScale(None=Sine(None=@n30) {period=2.557}), B=ConstantBiasScale(None=Sine(None=@n31)))
  n33 = Multiply(A=@n20.Alpha)
  n34 = Add(A=Multiply(A=Multiply(A=@n28.R, B=@n32)), B=Multiply(A=Clamp(None=@n33) {desc=falloff no contrast}) {constB=2})
  n35 = Desaturation(None=SceneColor().RGB)
  n36 = Multiply(A=@n35, B=@n35)
  n37 = Multiply(A=@n34, B=Multiply(A=@n36) {constB=100})
  n38 = LinearInterpolate(
   A=VectorParameter() {parameterName=Color2, defaultValue=<rgba 0.2222 0.092 0.4576 1>}.RGB
   B=VectorParameter() {parameterName=Color1, defaultValue=<rgba 0.9931 0.2074 1 1>}.RGB
   Alpha=Clamp(None=Add(A=@n34, B=@n37))
  )
  Multiply(A=@n24.RGB, B=@n38)
  Multiply(A=@n24.RGB)
  n39 = ComponentMask(None=@n11.Scale X) {r=true}
  Multiply(A=@n39) {constB=1.4, desc=max normalization error}
  Multiply(A=OneMinus(None=@n12)) {constB=2}
  n40 = Subtract(
   A=MaterialFunctionCall() {materialFunction=&WorldPositionBehindTranslucency}.Result
   B=@n1.Result
  )
  n41 = Multiply(A=Multiply(A=Normalize(VectorInput=@n40), B=@n19) {constA=200}, B=@n39) {constB=-200}
  n42 = MaterialFunctionCall() {materialFunction=&ObjectScale}
  n43 = Divide(A=Subtract(A=Add(A=@n41, B=@n10.Result), B=@n1.Result), B=Multiply(A=Add(A=@n42.Scale XYZ, B=@n39)) {constB=0.5}) {desc=0-1values inside the box}
  Multiply(A=@n43, B=@n43)
  n44 = OneMinus(None=Clamp(None=@n25))
  n45 = Multiply(A=@n44, B=@n44)
  n46 = Multiply(A=Multiply(A=Multiply(A=@n36) {constB=10}, B=@n38), B=Multiply(A=@n45, B=@n45))
  n47 = Add(A=Multiply(A=Multiply(A=@n46) {constB=0.5}, B=@n25), B=@n25)
  Multiply(A=@n47, B=@n47)
  ParticlePositionWS()
  ParticleSize()
 ]
}
